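; DeferredDecal_WaterDrips is a decal actor Blueprint, named for water drips, that applies configurable scalar parameters to its decal material.
; When cosmetic events can execute, the construction script creates and stores a dynamic material instance for the decal.
; It splits the Scalar Parameters map into keys and values, then sets each named parameter by matching index.
; BeginPlay, overlap and Tick events are empty.
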(event EventBeginPlay)

(event Collision|EventActorBeginOverlap
  (params (OtherActor :type Actor%20Object%20Reference)))

(event EventTick
  (params (DeltaSeconds :type Float%20%28single-precision%29)))

(fn ConstructionScript
  (Networking|CanExecuteCosmeticEvents
    (:True
      (bind _returnvalue (Rendering|Components|Decal|CreateDynamicMaterialInstance (Variables|Default|GetDecal)))
      (bind _output_get (Variables|Default|SetDecal%20MID _returnvalue))
      (bind _keys (Utilities|Map|Keys (Variables|Default|GetScalar%20Parameters)))
      (bind _values (Utilities|Map|Values (Variables|Default|GetScalar%20Parameters)))
      (Variables|Default|SetParam%20Values _values)
      (for (_array_element _array_index) _keys
        (Rendering|Material|SetScalarParameterValue _output_get _array_element (Utilities|Array|Get(aref) (Variables|Default|GetParam%20Values) _array_index) :!class /Script/Engine.MaterialInstanceDynamic)))))
